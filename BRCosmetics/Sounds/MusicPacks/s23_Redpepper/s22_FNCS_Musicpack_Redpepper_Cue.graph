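# Music pack sound cue for the Redpepper FNCS pack, played non-spatialized in the music pack sound class.
# A licensed content switcher selects either looping silence or the music mix, which fills both of its other inputs.
# The mix starts a one-shot intro and, after a short fixed delay, layers in the looping track.
# Concurrency allows one instance, stopping the oldest, and the output also feeds a music audio analysis submix.
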SoundCue /BRCosmetics/Sounds/MusicPacks/s23_Redpepper/s22_FNCS_Musicpack_Redpepper_Cue
root=FortSoundNodeLicensedContentSwitcher_0
nodes [
 {id=Delay_0, children [WavePlayer_2], properties {delayMax=0.277, delayMin=0.277}}
 {id=FortSoundNodeLicensedContentSwitcher_0, class=FortSoundNodeLicensedContentSwitcher, children [WavePlayer_0,Mixer_0,Mixer_0]}
 {id=Mixer_0, children [WavePlayer_1,Delay_0], properties {inputVolume [1,1]}}
 {id=WavePlayer_0, properties {bLooping=true, soundWaveAssetPtr=/Game/Sounds/Fort_Test_Sounds/5_seconds_silence}}
 {id=WavePlayer_1, properties {soundWaveAssetPtr=/Game/Athena/Sounds/MusicPacks/s23_Redpepper/MusicPack_RedPepper_Intro}}
 {id=WavePlayer_2, properties {bLooping=true, soundWaveAssetPtr=/Game/Athena/Sounds/MusicPacks/s23_Redpepper/MusicPack_RedPepper_Loop}}
]
properties {
 maxDistance=2097152
 soundClassObject=/Game/Sounds/FortSoundClassesAndModes/Fort_Music_MusicPacks
 volumeMultiplier=1
 attenuationOverrides {bAttenuate=false, bSpatialize=false}
 concurrencyOverrides {maxCount=1, resolutionRule=StopOldest}
 soundSubmixSends [{minSendLevel=0, soundSubmix=&Fortnite_AudioAnalysis_Music}]
}
